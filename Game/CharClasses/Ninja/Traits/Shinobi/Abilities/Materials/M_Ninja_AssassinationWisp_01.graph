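# Masked, two-sided lit mesh-particle material for the Ninja Shinobi assassination wisp effect.
# Panned beam textures, their time scaled by the RandomTime dynamic parameter, form streaks that feed emissive color and an apparent opacity mask.
# A second emissive layer adds sharpened panning noise, with each layer's color scaled by its own particle parameter.
# The Sweep parameter thresholds a U gradient against two edges to wipe the streaks along the mesh.
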Material /Game/CharClasses/Ninja/Traits/Shinobi/Abilities/Materials/M_Ninja_AssassinationWisp_01
graph {
 properties {
  bUsedWithMeshParticles=true
  blendMode=BLEND_Masked
  opacityMaskClipValue=0.5
  pixelDepthOffsetMode=PDOM_Legacy
  twoSided=true
  usedShadingModels=Default Lit
 }
 nodes [
  n1 = Panner() {bFractionalPart=true, speedX=-5.18, speedY=0.6}
  n2 = Multiply(
   A=Power(Base=TextureSample(UVs=@n1) {texture=&T_NoisePack_01}.G) {constExponent=4}
   B=VectorParameter() {parameterName=Emissive Color B, defaultValue=<rgba 0.5 8.75 50 0>}.RGB
  )
  n3 = DynamicParameter() {paramNames [RandomTime,Sweep,EmissiveA,EmissiveB]}
  n4 = TextureCoordinate() {uTiling=1.22}
  n5 = Constant2Vector() {g=0.09}
  n6 = Multiply(A=DynamicParameter() {paramNames [RandomTime,Sweep,EmissiveA,EmissiveB]}.RandomTime, B=Time())
  n7 = TextureSample(UVs=Panner(Coordinate=Add(A=@n4, B=@n5), Time=@n6) {bFractionalPart=true, speedX=-3.113}) {samplerType=SAMPLERTYPE_LinearColor, texture=/Game/Effects/Fort_Effects/Textures/Tile/T_Raygun_Beam_01}
  n8 = TextureCoordinate() {uTiling=1.418}
  n9 = Constant2Vector() {g=-0.04}
  n10 = TextureSample(UVs=Panner(Coordinate=Add(A=@n8, B=@n9), Time=@n6) {bFractionalPart=true, speedX=-3.84}) {samplerType=SAMPLERTYPE_LinearColor, texture=&T_Raygun_Beam_01}
  n11 = Multiply(
   A=Add(A=Multiply(A=@n7.G, B=@n10.G), B=@n10.B)
   B=VectorParameter() {parameterName=Emissive Color A, defaultValue=<rgba 21 95.5833 200 0>}.RGB
  )
  n12 = DynamicParameter() {paramNames [RandomTime,Sweep,EmissiveA,EmissiveB]}
  MaterialFunctionCall(Emissive=Add(A=Multiply(A=@n2, B=@n3.EmissiveB), B=Multiply(A=@n11, B=@n12.EmissiveA))) {materialFunction=&MF_FX_EmissiveAdjustments}
  n13 = Multiply(A=TextureCoordinate() {uTiling=3.45}, B=Constant2Vector() {g=0.7, r=1})
  n14 = Constant2Vector() {g=0.12}
  n15 = TextureSample(UVs=Panner(Coordinate=Add(A=@n13, B=@n14), Time=@n6) {bFractionalPart=true, speedX=-1.8}) {samplerType=SAMPLERTYPE_LinearColor, texture=&T_Raygun_Beam_01}
  n16 = TextureSample(UVs=Panner(Coordinate=TextureCoordinate() {uTiling=2.77}, Time=@n6) {bFractionalPart=true, speedX=-2.44}) {samplerType=SAMPLERTYPE_LinearColor, texture=&T_Raygun_Beam_01}
  n17 = MaterialFunctionCall() {materialFunction=&LinearGradient}
  n18 = DynamicParameter() {paramNames [RandomTime,Sweep,EmissiveA,EmissiveB]}
  n19 = Constant()
  n20 = If(A=@n17.UGradient, B=@n18.Sweep, "A > B"=Constant() {r=1}, "A == B"=@n19, "A < B"=@n19)
  n21 = Constant() {r=1}
  n22 = If(A=Add(A=@n17.UGradient), B=@n18.Sweep, "A > B"=Constant(), "A == B"=@n21, "A < B"=@n21)
  Multiply(A=Add(A=@n10.B, B=Add(A=@n15.A, B=@n16.B)), B=OneMinus(None=Add(A=@n20, B=@n22)))
  ScalarParameter() {parameterName=Metallic}
  ScalarParameter() {parameterName=Specular}
  ScalarParameter() {parameterName=Roughness}
  VectorParameter() {parameterName=BaseColor, defaultValue=<rgba 0.02 0.02 0.02 0>}
 ]
}
