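# Translucent, unlit, two-sided mesh-particle material for the smoky wisp of the Ninja's assassination ability.
# A packed thin-trail smoke texture pans along U, with speed and tiling driven by dynamic parameters, and an Image Random value picks among its RGB channels.
# That mask is multiplied by a tunable Fresnel term and particle alpha, while particle color feeds an emissive adjustment function.
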Material /Game/CharClasses/Ninja/Traits/Shinobi/Abilities/Materials/M_Ninja_AssassinationWisp_02
graph {
 properties {
  bUsedWithMeshParticles=true
  blendMode=BLEND_Translucent
  pixelDepthOffsetMode=PDOM_Legacy
  previewMesh=/Game/CharClasses/Ninja/Traits/Shinobi/Abilities/Meshes/Sm_Ninja_AssassinationWisp_01
  shadingModel=MSM_Unlit
  twoSided=true
  usedShadingModels=Unlit
 }
 nodes [
  n1 = ParticleColor()
  MaterialFunctionCall(Emissive=@n1.RGB) {materialFunction=&MF_FX_EmissiveAdjustments}
  n2 = DynamicParameter() {defaultValue=<rgba 0.99 1 1 1>, paramNames [Image Random, Panner Speed, Panner Tile, Param4]}
  n3 = TextureCoordinate()
  n4 = DynamicParameter() {defaultValue=<rgba 0.99 1 1 1>, paramNames [Image Random, Panner Speed, Panner Tile, Param4]}
  n5 = AppendVector(A=@n4.Panner Tile, B=Constant() {r=1})
  n6 = Time()
  n7 = TextureSample(UVs=Panner(Coordinate=Multiply(A=@n3, B=@n5), Time=Multiply(A=@n4.Panner Speed, B=@n6)) {speedX=-1}) {texture=/Game/Effects/Fort_Effects/Textures/Smoke/T_Smoke_ThinTrail_Packed}
  n8 = ScalarParameter() {defaultValue=1.45, parameterName=Fresnel Brightness}
  n9 = ScalarParameter() {defaultValue=3, parameterName=Fresnel Exponent}
  Multiply(
   A=@n1.A
   B=Multiply(
    A=MaterialFunctionCall!(Random Picker=@n2.Image Random, Texture Input=@n7.RGB, Option 1, Option 2, Option 3) {materialFunction=/Game/Effects/Fort_Effects/Materials/_Functions/MF_RGB_Random_01}.Result
    B=MaterialFunctionCall(Brightness=@n8, Exponent=@n9) {materialFunction=&MF_Fresnel_FX}.Fresnel
   )
  )
 ]
}
